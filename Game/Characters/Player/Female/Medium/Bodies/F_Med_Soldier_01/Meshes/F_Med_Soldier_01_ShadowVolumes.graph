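# Physics asset of capsule shadow volumes for the female medium Soldier 01 player body, a rough proxy of the character's shape.
# One capsule per body covers pelvis, upper spine, arms, legs and feet, with a tiny one on spine_01 and none for head or hands.
# Constraints form a simplified tree: feet attach to thighs and forearms to spine_04, with only the forearm joints explicitly free to swing and twist.
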PhysicsAsset /Game/Characters/Player/Female/Medium/Bodies/F_Med_Soldier_01/Meshes/F_Med_Soldier_01_ShadowVolumes
bodies {
 calf_twist_02_l [{shapeType=Capsule, length=48.6942, radius=5.244, center=<-0.8061 1.9657 -0.0034>, rotation=<rot -4.5152 96.6603 -88.9489>}]
 calf_twist_02_r [{shapeType=Capsule, length=43.5231, radius=6.1316, center=<3.4576 0.8977 0.1612>, rotation=<rot -1.1437 89.8273 -87.9177>}]
 foot_l [{shapeType=Capsule, length=13.2614, radius=5.0479, center=<-2.7225 6.8878 -0.3789>, rotation=<rot -0.2222 -178.341 -90.7451>}]
 foot_r [{shapeType=Capsule, length=11.5114, radius=7.2979, center=<1.0279 -7.3599 0.3828>, rotation=<rot 0.2149 2.2179 -89.2527>}]
 lowerarm_l [{shapeType=Capsule, length=32.4607, radius=4.1998, center=<15.2261 0.1021 -0.4193>, rotation=<rot -21.4684 -89.9999 -88.4281>}]
 lowerarm_r [{shapeType=Capsule, length=27.4607, radius=4.9498, center=<-15.0788 -0.2772 0.8631>, rotation=<rot -21.4684 -90 -84.3865>}]
 pelvis [{shapeType=Capsule, length=5.4184, radius=15.2636, center=<0.6811 -4.9851 -1.437>, rotation=<rot 6.0822 -90 1.426>}]
 spine_01 [{shapeType=Capsule, length=0.505, radius=0.505, rotation=<rot 0 88.1815 -90>}]
 spine_04 [{shapeType=Capsule, length=17.1651, radius=13.8902, center=<-0.457 2.064 -0.2153>, rotation=<rot 0 90 -90>}]
 thigh_l [{shapeType=Capsule, length=21.2632, radius=8.8667, center=<-19.1716 0.8545 -0.8627>, rotation=<rot 0 85.2774 -88.5772>}]
 thigh_r [{shapeType=Capsule, length=14.2962, radius=8.64, center=<28.5796 0.0368 0.2364>, rotation=<rot 0 -81.8699 -87.6417>}]
 upperarm_twist_01_l [{shapeType=Capsule, length=21.4137, radius=4.7588, center=<4.0075 -0.2242 -0.3003>, rotation=<rot 0 -94.0203 -91.3636>}]
 upperarm_twist_01_r [{shapeType=Capsule, length=23.685, radius=4.4112, center=<-3.448 0.1722 0.1272>, rotation=<rot -4.5917 -92.687 -92.0039>}]
}
constraints [
 {bone1Name=spine_01, bone2Name=pelvis, swing1LimitDegrees=45, swing2LimitDegrees=45, twistLimitDegrees=45}
 {bone1Name=spine_04, bone2Name=spine_01, swing1LimitDegrees=45, swing2LimitDegrees=45, twistLimitDegrees=45}
 {bone1Name=upperarm_twist_01_l, bone2Name=spine_04, swing1LimitDegrees=45, swing2LimitDegrees=45, twistLimitDegrees=45}
 {bone1Name=upperarm_twist_01_r, bone2Name=spine_04, swing1LimitDegrees=45, swing2LimitDegrees=45, twistLimitDegrees=45}
 {bone1Name=thigh_l, bone2Name=pelvis, swing1LimitDegrees=45, swing2LimitDegrees=45, twistLimitDegrees=45}
 {bone1Name=calf_twist_02_l, bone2Name=thigh_l, swing1LimitDegrees=45, swing2LimitDegrees=45, twistLimitDegrees=45}
 {bone1Name=foot_l, bone2Name=thigh_l, swing1LimitDegrees=45, swing2LimitDegrees=45, twistLimitDegrees=45}
 {bone1Name=thigh_r, bone2Name=pelvis, swing1LimitDegrees=45, swing2LimitDegrees=45, twistLimitDegrees=45}
 {bone1Name=calf_twist_02_r, bone2Name=thigh_r, swing1LimitDegrees=45, swing2LimitDegrees=45, twistLimitDegrees=45}
 {bone1Name=foot_r, bone2Name=thigh_r, swing1LimitDegrees=45, swing2LimitDegrees=45, twistLimitDegrees=45}
 {
  bone1Name=lowerarm_r
  bone2Name=spine_04
  swing1LimitDegrees=45
  swing1Motion=Free
  swing2LimitDegrees=45
  swing2Motion=Free
  twistLimitDegrees=45
  twistMotion=Free
 }
 {
  bone1Name=lowerarm_l
  bone2Name=spine_04
  swing1LimitDegrees=45
  swing1Motion=Free
  swing2LimitDegrees=45
  swing2Motion=Free
  twistLimitDegrees=45
  twistMotion=Free
 }
]
