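# Physics asset for the legs of the female medium Soldier 01 player body, covering dangling left calf straps and right boot strings.
# Small capsule bodies on two straps and two strings are each constrained to their side's lower calf twist bone, with narrow twist limits and differing swing limits.
# Both calf twist bodies link to a pelvis body through constraints with free swing and twist, leaving their rotation unrestricted.
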PhysicsAsset /Game/Characters/Player/Female/Medium/Bodies/F_Med_Soldier_01/Meshes/F_MED_Soldier_01_Dyn_legs
bodies {
 calf_straps_bott_l [{shapeType=Capsule, length=3, radius=2, center=<0.3 0 0>, rotation=<rot 90 0 0>}]
 calf_straps_top_l [{shapeType=Capsule, length=3, radius=2, center=<0.3 0 0>, rotation=<rot 90 0 0>}]
 calf_twist_02_l [{shapeType=Capsule, length=0.5, radius=0.5, rotation=<rot 90 0 0>}]
 calf_twist_02_r [{shapeType=Capsule, length=0.5, radius=0.5, rotation=<rot 90 0 0>}]
 dyn_boot_string_in_r [{shapeType=Capsule, length=3, radius=2, center=<1.069 -0.0001 0>, rotation=<rot 90 0 0>}]
 dyn_boot_string_out_r [{shapeType=Capsule, length=3, radius=2, center=<1.5858 0 0>, rotation=<rot 90 0 0>}]
 pelvis [{shapeType=Capsule, length=3, radius=2, rotation=<rot 90 0 0>}]
}
constraints [
 {bone1Name=calf_straps_bott_l, bone2Name=calf_twist_02_l, swing1LimitDegrees=45, swing2LimitDegrees=3, twistLimitDegrees=2}
 {bone1Name=calf_straps_top_l, bone2Name=calf_twist_02_l, swing1LimitDegrees=45, swing2LimitDegrees=3, twistLimitDegrees=2}
 {bone1Name=dyn_boot_string_in_r, bone2Name=calf_twist_02_r, swing1LimitDegrees=30, swing2LimitDegrees=45, twistLimitDegrees=2}
 {bone1Name=dyn_boot_string_out_r, bone2Name=calf_twist_02_r, swing1LimitDegrees=30, swing2LimitDegrees=45, twistLimitDegrees=2}
 {
  bone1Name=calf_twist_02_l
  bone2Name=pelvis
  swing1LimitDegrees=45
  swing1Motion=Free
  swing2LimitDegrees=45
  swing2Motion=Free
  twistLimitDegrees=45
  twistMotion=Free
 }
 {
  bone1Name=calf_twist_02_r
  bone2Name=pelvis
  swing1LimitDegrees=45
  swing1Motion=Free
  swing2LimitDegrees=45
  swing2Motion=Free
  twistLimitDegrees=45
  twistMotion=Free
 }
]
